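# Translucent lit material for Niagara mesh particles that draws wind or smoke streaks on thickened splines for the Rift Gate effect.
# Dynamic parameters drive SplineThicken width, erosion and random offsets for panning smoke color and mask textures.
# A sphere mask travels along UVs over particle lifetime, combining with edge falloff and erosion to shape opacity.
# Depth fade softens intersections, replaced by particle alpha times a mobile opacity on forward and mobile paths.
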Material /PlotTwist/Items/RiftGate/Assets/FX/M_PlotTwist_RiftGate_Spline_Particles
graph {
 properties {
  bUsedWithNiagaraMeshParticles=true
  blendMode=BLEND_Translucent
  pixelDepthOffsetMode=PDOM_Legacy
  twoSided=true
  usedShadingModels=Default Lit
 }
 nodes [
  n1 = DynamicParameter() {paramNames [Erode,SplineWidth,PanAdd,Random]}
  n2 = Multiply(A=@n1.SplineWidth, B=ScalarParameter() {defaultValue=20, parameterName=SplineWidth})
  MaterialFunctionCall(WidthBase=Multiply(A=ScalarParameter() {defaultValue=0.7, parameterName=SplineWidthTipMult}, B=@n2) {constB=5}, WidthTip=@n2) {materialFunction=&SplineThicken}
  n3 = ParticleColor()
  n4 = TextureCoordinate() {vTiling=0.5}
  n5 = Add(A=@n1.Random, B=VertexColor().R)
  n6 = Time()
  n7 = Panner(
   Coordinate=MaterialFunctionCall(XY=@n4) {materialFunction=&Swizzle}.YX
   Time=Add(A=@n1.PanAdd, B=Add(A=@n5, B=@n6))
  ) {speedX=0.5}
  n8 = TextureSampleParameter2D(UVs=@n7) {parameterName=ColorTexture, texture=/Game/Effects/Fort_Effects/Textures/Smoke/T_Smoke_Trail_Soft_Chromatic}
  n9 = Constant() {r=1}
  MaterialFunctionCall(Emissive=Multiply(A=@n3.RGB, B=StaticSwitchParameter(True=@n8.RGB, False=@n9) {parameterName=UseColorTexture})) {materialFunction=&MF_FX_EmissiveAdjustments}
  n10 = LinearInterpolate(B=@n3.RGB) {constA=1, constAlpha=0.3}
  n11 = Add(A=ComponentMask(None=@n4) {r=true}) {constB=0.15}
  n12 = TextureSampleParameter2D(UVs=@n7) {parameterName=Mask, samplerType=SAMPLERTYPE_Grayscale, texture=&T_Smoke_Trail_Soft}
  n13 = LinearInterpolate(Alpha=Add(A=Multiply(A=@n1.Erode, B=@n11), B=@n12.R)) {constA=-3}
  n14 = TextureCoordinate()
  n15 = LinearInterpolate(Alpha=VertexColor().R) {constA=0.8, constB=1.2}
  n16 = ScalarParameter() {parameterName=RVertPaintOffset}
  n17 = Power(Base=Multiply(A=ParticleRelativeTime(), B=LinearInterpolate(B=@n15, Alpha=@n16)))
  n18 = SphereMask(
   A=@n14
   B=AppendVector(A=LinearInterpolate(Alpha=@n17) {constA=0.97, constB=-0.7}, B=Constant() {r=0.5})
   Radius=LinearInterpolate(Alpha=@n17) {constA=0.2}
  ) {attenuationRadius=0.5, hardnessPercent=0}
  n19 = ScalarParameter() {defaultValue=1, parameterName=UseTravel}
  n20 = ComponentMask(None=@n14) {g=true}
  n21 = Constant() {r=0.5}
  n22 = ScalarParameter() {defaultValue=0.5, parameterName=StartEndHardness}
  n23 = Multiply(
   A=Clamp(None=Multiply(A=@n13, B=@n12.R))
   B=Multiply(A=LinearInterpolate(B=@n18, Alpha=@n19) {constA=1}, B=SphereMask(A=@n20, B=@n21, Hardness=@n22) {attenuationRadius=0.5, hardnessPercent=0.5}) {constA=1}
  )
  Multiply(
   A=Multiply(A=VectorParameter() {parameterName=WindColor, defaultValue=<rgba 0.3594 0.3436 0.3167 1>}.RGB, B=Clamp(None=@n10) {minDefault=0.1})
   B=Add(A=@n23) {constB=0.1}
  )
  n24 = Multiply(A=@n3.A, B=@n23)
  n25 = ScalarParameter() {defaultValue=40, parameterName=DepthFadeDist}
  n26 = Multiply(A=ScalarParameter() {defaultValue=1, parameterName=Mobile Opacity}, B=ParticleColor().A)
  Multiply(A=@n24, B=ShadingPathSwitch(Default=DepthFade(Opacity=@n24, FadeDistance=@n25) {fadeDistanceDefault=55}, Forward=@n26, Mobile=@n26))
 ]
}
