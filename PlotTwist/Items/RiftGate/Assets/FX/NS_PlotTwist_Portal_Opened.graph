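# Niagara effect for the opened Rift Gate portal, layering a flickering ground sprite and glow, light beams, motes, sparkles, shimmers and side mesh strips.
# OpenAmount widens spawn shapes and mesh offsets, MasterOpacity fades several layers, and InteriorColor tints beams and particles.
# WeldProgress shortens beams and brightens sparkles.
# Beams align along lines from a distant owner-relative point, motes swirl in a vortex, and the light-attribute module is disabled.
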NiagaraSystem /PlotTwist/Items/RiftGate/Assets/FX/NS_PlotTwist_Portal_Opened
dependencies {
 usedDataInterfaces [
  &NiagaraDataInterfaceCurve
  &NiagaraDataInterfaceMeshRendererInfo
  &NiagaraDataInterfaceColorCurve
 ]
 usedDynamicInputs [&RandomRangeVector]
 usedModules [
  &SystemState
  &self:SystemSpawnScript.SystemScriptSource.SystemScriptGraph.NiagaraNodeAssignment_0.SetVariables_#1
 ]
}
emitters {
 GroundSprite {
  emitterUpdateScript [
   &EmitterState()
   &SpawnBurst_Instantaneous(Use Spawn Probability=false, UseLoopCountLimit=true)
  ]
  particleSpawnScript [
   /Niagara/Modules/Spawn/Initialization/InitializeParticle(
    Color=<rgba 0 1.3509 3 0.2>
    Lifetime=4
    Module.Write Color=true
    Module.Write Lifetime=true
    Module.Write Mass=false
    Module.Write Material Random=true
    Module.Write Position=true
    Module.Write Scale=true
    Module.Write SpriteRotation=true
    Module.Write SpriteSize=true
    Position=<0 0 20>
    Sprite Rotation=90
    Sprite Size=<500 800>
   )
   &DynamicMaterialParameters(
    Index 0 Param 1=1.5
    Index 0 Param 2=1
    Module.Param0WriteEnabled=true
    Module.Param1WriteEnabled=true
    Module.Param2WriteEnabled=false
    Module.Param3WriteEnabled=false
   )
   &CameraOffset(Camera Offset Amount=50)
   &SpriteFacingAndAlignment(Sprite Alignment=<1 0 0>, Sprite Facing=<0 0 1>, WriteAlignment=true, WriteFacing=true)
  ]
  particleUpdateScript [
   &ParticleState(
    Kill Particles When Lifetime Has Elapsed=false
    Let Infinitely Lived Particles Die When Emitter Deactivates=true
    Loop Particles Lifetime=true
   )
   &ScaleSpriteSize(Scale Sprite Size Mode=::NewEnumerator0, Uniform Scale Factor=&Lerp_Float(A=0.35, B=1, Alpha=$User.OpenAmount))
   &ScaleColor(
    Scale Alpha=&FloatFromCurve(
     FloatCurve=DataInterface() {propertyValues {Curve {keys [
         <at 0 1 RCIM_Cubic>
         {arriveTangent=0.2795, interpMode=RCIM_Cubic, leaveTangent=0.2795, time=0.0763, value=0.9756}
         {arriveTangent=-0.0227, interpMode=RCIM_Cubic, leaveTangent=-0.0227, tangentMode=RCTM_User, time=0.1816, value=1.0507}
         {arriveTangent=-0.4352, interpMode=RCIM_Cubic, leaveTangent=-0.4352, time=0.2195, value=0.8992}
         {arriveTangent=0.4599, interpMode=RCIM_Cubic, leaveTangent=0.4599, time=0.3016, value=0.9985}
         {arriveTangent=0.2826, interpMode=RCIM_Cubic, leaveTangent=0.2826, time=0.3821, value=0.974}
         {arriveTangent=-0.3797, interpMode=RCIM_Cubic, leaveTangent=-0.3797, time=0.4508, value=1.0407}
         {arriveTangent=-0.4815, interpMode=RCIM_Cubic, leaveTangent=-0.4815, time=0.4836, value=0.9355}
         {arriveTangent=-0.2663, interpMode=RCIM_Cubic, leaveTangent=-0.2663, time=0.5406, value=0.9975}
         {arriveTangent=0.0945, interpMode=RCIM_Cubic, leaveTangent=0.0945, time=0.5923, value=0.9065}
         {arriveTangent=0.4382, interpMode=RCIM_Cubic, leaveTangent=0.4382, time=0.6803, value=1.0107}
         {arriveTangent=0.4969, interpMode=RCIM_Cubic, leaveTangent=0.4969, time=0.7689, value=0.9839}
         {arriveTangent=-0.2345, interpMode=RCIM_Cubic, leaveTangent=-0.2345, time=0.8405, value=1.0903}
         {arriveTangent=-0.5661, interpMode=RCIM_Cubic, leaveTangent=-0.5661, time=0.8922, value=0.955}
         <at 1 1 RCIM_Cubic>
        ]}}}
     Scale Curve=$User.MasterOpacity
    )
    ScaleA=true
    ScaleRGB=true
   )
  ]
  rendererClasses [&NiagaraSpriteRendererProperties]
  renderers [{rendererClass=&NiagaraSpriteRendererProperties}]
 }
 Light {
  emitterUpdateScript [
   &EmitterState()
   &SpawnBurst_Instantaneous(Use Spawn Probability=false, UseLoopCountLimit=true)
  ]
  particleSpawnScript [
   &InitializeParticle(
    Color=<rgba 0.1586 0.6176 0.9823 0.35>
    Lifetime=4
    Module.Write Color=true
    Module.Write Lifetime=true
    Module.Write Mass=false
    Module.Write Material Random=true
    Module.Write Position=true
    Module.Write Scale=true
    Module.Write SpriteRotation=false
    Module.Write SpriteSize=true
    Position=<0 0 150>
    Sprite Size=<500 1000>
   )
   &DynamicMaterialParameters(
    Index 0 Param 1=1.5
    Index 0 Param 2=1
    Module.Param0WriteEnabled=true
    Module.Param1WriteEnabled=true
    Module.Param2WriteEnabled=false
    Module.Param3WriteEnabled=false
   )
   &CameraOffset(Camera Offset Amount=-50)
  ]
  particleUpdateScript [
   &ParticleState(
    Kill Particles When Lifetime Has Elapsed=false
    Let Infinitely Lived Particles Die When Emitter Deactivates=true
    Loop Particles Lifetime=true
   )
   &ScaleColor(
    Scale Alpha=&FloatFromCurve(
     FloatCurve=DataInterface() {propertyValues {Curve {keys [
         <at 0 1 RCIM_Cubic>
         {arriveTangent=0.2795, interpMode=RCIM_Cubic, leaveTangent=0.2795, time=0.0763, value=0.9756}
         {arriveTangent=-0.0227, interpMode=RCIM_Cubic, leaveTangent=-0.0227, tangentMode=RCTM_User, time=0.1816, value=1.0507}
         {arriveTangent=-0.4352, interpMode=RCIM_Cubic, leaveTangent=-0.4352, time=0.2195, value=0.8992}
         {arriveTangent=0.4599, interpMode=RCIM_Cubic, leaveTangent=0.4599, time=0.3016, value=0.9985}
         {arriveTangent=0.2826, interpMode=RCIM_Cubic, leaveTangent=0.2826, time=0.3821, value=0.974}
         {arriveTangent=-0.3797, interpMode=RCIM_Cubic, leaveTangent=-0.3797, time=0.4508, value=1.0407}
         {arriveTangent=-0.4815, interpMode=RCIM_Cubic, leaveTangent=-0.4815, time=0.4836, value=0.9355}
         {arriveTangent=-0.2663, interpMode=RCIM_Cubic, leaveTangent=-0.2663, time=0.5406, value=0.9975}
         {arriveTangent=0.0945, interpMode=RCIM_Cubic, leaveTangent=0.0945, time=0.5923, value=0.9065}
         {arriveTangent=0.4382, interpMode=RCIM_Cubic, leaveTangent=0.4382, time=0.6803, value=1.0107}
         {arriveTangent=0.4969, interpMode=RCIM_Cubic, leaveTangent=0.4969, time=0.7689, value=0.9839}
         {arriveTangent=-0.2345, interpMode=RCIM_Cubic, leaveTangent=-0.2345, time=0.8405, value=1.0903}
         {arriveTangent=-0.5661, interpMode=RCIM_Cubic, leaveTangent=-0.5661, time=0.8922, value=0.955}
         <at 1 1 RCIM_Cubic>
        ]}}}
     Scale Curve=$User.MasterOpacity
    )
    ScaleA=true
    ScaleRGB=true
   )
   &FN_Light_Attributes(
    Color=$Particles.Initial.Color
    Create Color Attribute=true
    Create Exponent Attribute=true
    Create Position Attribute=true
    Create Radius Attribute=true
    Exponent=2
    Position=<0 0 50>
    Radius=&Multiply_Float(A=300, B=$User.MasterOpacity)
    Scale Color=&Multiply_Float(A=10, B=$User.MasterOpacity)
   ) {enabled=false}
  ]
  rendererClasses [&NiagaraSpriteRendererProperties, &NiagaraLightRendererProperties]
  renderers [
   {rendererClass=&NiagaraSpriteRendererProperties}
   {rendererClass=&NiagaraLightRendererProperties, rendererIndex=1}
  ]
 }
 LightBeams {
  emitterUpdateScript [
   &EmitterState()
   &SpawnRate(SpawnRate=10, Use Spawn Probability=false)
  ]
  particleSpawnScript [
   /Niagara/Modules/Spawn/Initialization/V2/InitializeParticle(
    Color=<rgba 1 1 1 0.4>
    Mesh Scale Mode=::NewEnumerator3
    Mesh Uniform Scale=5
    Position Mode=::NewEnumerator1
    Sprite Size Max=<110 1200>
    Sprite Size Min=<80 800>
    Sprite Size Mode=::NewEnumerator2
    UsePositionOffset=false
    Write Material Random=true
   )
   &ShapeLocation(
    MaskSpawnGroup=false
    Non Uniform Scale=&Lerp_Vector(A=&Lerp_Vector(A=<0.5 1 1>, B=<0.5 22 1>, Alpha=$User.OpenAmount), B=<0.5 1 1>, Alpha=$Particles.MaterialRandom)
    Offset=&MakeVector(X=0, Y=0, Z=&Lerp_Float(A=50, B=450, Alpha=$Particles.MaterialRandom))
    Offset Mode=::NewEnumerator0
    Override Seed=false
    Shape Origin=$Engine.Owner.Position
    Sphere Radius=10
    UseOwnerScale=true
   )
   &DynamicMaterialParameters(
    Index 0 Param 1=&RandomRangeFloat(Minimum=-0.1, Maximum=-0.05)
    Index 0 Param 2=1
    Index 0 Param 3=0.5
    Module.Param0WriteEnabled=true
    Module.Param1WriteEnabled=false
    Module.Param2WriteEnabled=false
    Module.Param3WriteEnabled=false
   )
  ]
  particleUpdateScript [
   &ParticleState()
   /SetVariables(Particles.SpriteAlignment=&NormalizeVector(
     Vector To Normalize=&Subtract_Vector(
      A=Linked() {linkedVariable {name=Particles.Position, type=NiagaraPosition}}
      B=&Add_Vector(
       A=Linked() {linkedVariable {name=Engine.Owner.Position, type=NiagaraPosition}}
       B=&RotateVectorByQuaternion(Vector To Rotate=<12000 0 8000>, Quaternion=$Engine.Owner.Rotation)
      )
     )
     Safe Normalize=true
     Safe Fallback Threshold=0
     Safe Fallback Vector=<0 0 0.001>
    )) {isSetParameters=true}
   &ScaleColor(
    Scale Alpha=&FloatFromCurve(
     FloatCurve=DataInterface() {propertyValues {Curve {keys [<at 0 0 RCIM_Cubic>, {interpMode=RCIM_Cubic, tangentMode=RCTM_User, time=0.5, value=1}, <at 1 0>]}}}
     Scale Curve=$User.MasterOpacity
    )
    Scale RGB=&MakeVectorFromLinearColorRGB(Color=$User.InteriorColor)
    ScaleA=true
    ScaleRGB=true
   )
   &ScaleSpriteSize(Scale Factor=&MakeVector2D(
     X=&Lerp_Float(A=0.5, B=1, Alpha=$System.OpenAmount)
     Y=&FloatFromCurve(
      FloatCurve=DataInterface() {propertyValues {Curve {keys [
          <at 0 0.75 RCIM_Cubic>
          <at 0.4832 1 RCIM_Cubic>
          {arriveTangent=-1.067, interpMode=RCIM_Cubic, leaveTangent=-1.067, tangentMode=RCTM_User, time=1, value=0.75}
         ]}}}
      Scale Curve=&Lerp_Float(A=1, B=0.2, Alpha=$User.WeldProgress)
     )
    ))
  ]
  rendererClasses [&NiagaraSpriteRendererProperties]
  renderers [{rendererClass=&NiagaraSpriteRendererProperties}]
 }
 Motes {
  emitterUpdateScript [
   &EmitterState()
   &SpawnRate(
    Spawn Probability=$User.MasterOpacity
    SpawnRate=&Lerp_Float(A=40, B=15, Alpha=$System.OpenAmount)
    Use Spawn Probability=true
   )
  ]
  particleSpawnScript [
   &InitializeParticle(
    Color=&LerpLinearColors(
     StartColor=&RandomRangeLinearColor(Minimum=<rgba 0.1 1.206 3 1>, Maximum=<rgba 0.3154 4.3512 5 1>)
     EndColor=$User.InteriorColor
     LerpFactor=$System.OpenAmount
    )
    Lifetime Max=7
    Lifetime Min=3
    Lifetime Mode=::NewEnumerator1
    Mesh Scale Mode=::NewEnumerator4
    Position Mode=::NewEnumerator1
    Sprite Rotation Mode=::NewEnumerator2
    Sprite Size Mode=::NewEnumerator1
    Uniform Sprite Size Max=30
    Uniform Sprite Size Min=10
    UsePositionOffset=false
    Write Material Random=true
   )
   &ShapeLocation(
    MaskSpawnGroup=false
    Non Uniform Scale=&Lerp_Vector(A=&Lerp_Vector(A=<0.5 1 1>, B=<0.5 22 1>, Alpha=$User.OpenAmount), B=<0.5 1 1>, Alpha=$Particles.MaterialRandom)
    Offset=&MakeVector(X=0, Y=0, Z=&Lerp_Float(A=50, B=450, Alpha=$Particles.MaterialRandom))
    Offset Mode=::NewEnumerator0
    Override Seed=false
    Shape Origin=$Engine.Owner.Position
    Sphere Radius=10
    UseOwnerScale=true
   )
   /SetVariables(Particles.Sparkle:NiagaraFloat=&UniformAOrBFloat(A=0, Distribution Weight=0.2)) {isSetParameters=true}
   &AddVelocity(
    Additional Quaternion Rotation=false
    Rotation Mode=::NewEnumerator3
    Rotation Quaternion:Quat4f=$Engine.Owner.Rotation
    Velocity=&MakeVector(X=&Lerp_Float(A=-200, B=-125, Alpha=$Particles.MaterialRandom), Y=0, Z=0)
   )
   &DynamicMaterialParameters(
    Index 0 Param 1=1
    Index 0 Param 2=1
    Index 0 Param 3=0.5
    Module.Param0WriteEnabled=true
    Module.Param1WriteEnabled=true
    Module.Param2WriteEnabled=true
    Module.Param3WriteEnabled=true
   )
  ]
  particleUpdateScript [
   &ParticleState()
   &ScaleColor(
    Scale Alpha=&FloatFromCurve(
     FloatCurve=DataInterface() {propertyValues {Curve {keys [
         <at 0 0 RCIM_Cubic>
         {interpMode=RCIM_Cubic, tangentMode=RCTM_User, time=0.1, value=0.9914}
         {interpMode=RCIM_Cubic, tangentMode=RCTM_User, time=0.7, value=1}
         <at 1 0>
        ]}}}
     Scale Curve=0.25
    )
    ScaleA=true
    ScaleRGB=true
   )
   &ScaleSpriteSize(Scale Factor=&Vector2DFromFloat(Value=&FloatFromCurve(
      FloatCurve=DataInterface() {propertyValues {Curve {keys [
          <at 0 0 RCIM_Cubic>
          {arriveTangent=2.4312, interpMode=RCIM_Cubic, leaveTangent=2.4312, time=0.1483, value=1}
          {arriveTangent=-0.0468, interpMode=RCIM_Cubic, leaveTangent=-0.0468, time=0.2916, value=0.709}
          {arriveTangent=-0.2796, interpMode=RCIM_Cubic, leaveTangent=-0.2796, time=0.4899, value=0.984}
          {arriveTangent=0.7494, interpMode=RCIM_Cubic, leaveTangent=0.7494, tangentMode=RCTM_User, time=0.6851, value=0.599}
          <at 1 0>
         ]}}}
      Scale Curve=&Lerp_Float(A=0.5, B=1, Alpha=$System.OpenAmount)
     )))
   &VortexForce(
    Influence Falloff Radius=1600
    Origin Pull Amount=-0.5
    Use Influence Falloff=true
    Vortex Axis=<0 1 0>
    Vortex Force Amount=100
    Vortex Origin:Vector3f=&AddVectorToPosition(
     Position=$Engine.Owner.Position
     Vector=&RotateVectorByQuaternion(Vector To Rotate=<-500 0 200>, Quaternion=$Engine.Owner.Rotation)
    )
   )
   &VectorNoiseForce(Force Amount=&Lerp_Float(A=100, B=1200, Alpha=$System.OpenAmount), Module.Use Random Noise Time Interval=true)
   /SetVariables(Particles.SparkleSize:Vector2f=/Niagara/DynamicInputs/Add/Add_Vector2(
     A:Vector2f=$Particles.SpriteSize
     B:Vector2f=&Vector2DFromFloat(Value=&Multiply_Float(
       A=$Particles.Sparkle
       B=&FloatFromCurve(
        FloatCurve=DataInterface() {propertyValues {Curve {keys [
            <at 0.4 0.05 RCIM_Cubic>
            {arriveTangent=-4.4497, interpMode=RCIM_Cubic, leaveTangent=-4.4497, tangentMode=RCTM_User, time=0.45, value=1}
            <at 0.6 0.05 RCIM_Cubic>
           ]}}}
        Scale Curve=100
       )
      ))
    )) {isSetParameters=true}
   &Drag(Drag=2, Use Linear Drag=true, Use Rotational Drag=true)
   &GravityForce(Gravity=<0 0 -40>)
   &SolveForcesAndVelocity(Clamp Velocity=false, Limit Acceleration=false)
  ]
  rendererClasses [&NiagaraSpriteRendererProperties]
  renderers [
   {rendererClass=&NiagaraSpriteRendererProperties}
   {rendererClass=&NiagaraSpriteRendererProperties, rendererIndex=1}
  ]
 }
 Shimmers {
  emitterUpdateScript [
   &EmitterState()
   &SpawnRate(SpawnRate=10, Use Spawn Probability=false)
  ]
  particleSpawnScript [
   &InitializeParticle(
    Color=<rgba 0.4846 0.8794 1 0.3>
    Lifetime Max=3
    Lifetime Min=2
    Lifetime Mode=::NewEnumerator1
    Sprite Rotation Mode=::NewEnumerator2
    Sprite Size Mode=::NewEnumerator1
    Uniform Sprite Size Max=175
    Uniform Sprite Size Min=125
    UsePositionOffset=false
    Write Material Random=true
   )
   &ShapeLocation(
    MaskSpawnGroup=false
    Non Uniform Scale=&Lerp_Vector(A=&Lerp_Vector(A=<0.5 18 1>, B=<0.5 22 1>, Alpha=$User.OpenAmount), B=<0.5 1 1>, Alpha=$Particles.MaterialRandom)
    Offset=&MakeVector(X=0, Y=0, Z=&Lerp_Float(A=10, B=100, Alpha=$Particles.MaterialRandom))
    Offset Mode=::NewEnumerator0
    Override Seed=false
    Shape Origin=$Engine.Owner.Position
    Sphere Radius=10
    UseOwnerScale=true
   )
   &AddVelocity(
    Module.Use Falloff Distance=false
    Origin Offset=<0 200 100>
    Origin Offset Coordinate Space=::Simulation
    Override Seed=false
    Velocity Mode=::NewEnumerator1
    Velocity Origin=$Engine.Owner.Position
    Velocity Speed=100
   )
   &CameraOffset(Camera Offset Amount=5)
  ]
  particleUpdateScript [
   &ParticleState()
   &Drag(Drag=0.25, Use Linear Drag=true, Use Rotational Drag=true)
   &DynamicMaterialParameters(
    Index 0 Param 1=$Particles.MaterialRandom
    Index 0 Param 2=&RandomRangeFloat(Minimum=0, Maximum=1)
    Module.Param0WriteEnabled=true
    Module.Param1WriteEnabled=true
    Module.Param2WriteEnabled=true
    Module.Param3WriteEnabled=false
   )
   &ScaleColor(
    Scale Alpha=&FloatFromCurve(
     FloatCurve=DataInterface() {propertyValues {Curve {keys [
         {arriveTangent=4.0377, interpMode=RCIM_Cubic, leaveTangent=4.0377, tangentMode=RCTM_User, time=0, value=0}
         <at 0.35 1 RCIM_Cubic>
         {arriveTangent=-2.0761, interpMode=RCIM_Cubic, leaveTangent=-2.0761, tangentMode=RCTM_User, time=1, value=0}
        ]}}}
     Scale Curve=&Lerp_Float(A=0.15, B=0.25, Alpha=$System.OpenAmount)
    )
    ScaleA=true
    ScaleRGB=true
   )
   &ScaleSpriteSize(Scale Factor=&Vector2DFromFloat(Value=&FloatFromCurve(
      FloatCurve=DataInterface() {propertyValues {Curve {keys [<at 0 0.7 RCIM_Cubic>, {interpMode=RCIM_Cubic, tangentMode=RCTM_User, time=1, value=1}]}}}
      Scale Curve=&Lerp_Float(A=0.75, B=1, Alpha=$System.OpenAmount)
     )))
   &SolveForcesAndVelocity(Clamp Velocity=false, Limit Acceleration=false)
  ]
  rendererClasses [&NiagaraSpriteRendererProperties]
  renderers [{rendererClass=&NiagaraSpriteRendererProperties}]
 }
 Sparkles {
  emitterUpdateScript [
   &EmitterState()
   &SpawnRate(
    Spawn Probability=$User.MasterOpacity
    SpawnRate=&Lerp_Float(A=15, B=35, Alpha=$System.OpenAmount)
    Use Spawn Probability=true
   )
  ]
  particleSpawnScript [
   &InitializeParticle(
    Color=&LerpLinearColors(
     StartColor=&RandomRangeLinearColor(Minimum=<rgba 0.1 1.206 3 1>, Maximum=<rgba 0.3154 4.3512 5 1>)
     EndColor=$User.InteriorColor
     LerpFactor=$System.OpenAmount
    )
    Lifetime Max=4
    Lifetime Min=1.25
    Lifetime Mode=::NewEnumerator1
    Mesh Scale Mode=::NewEnumerator4
    Position Mode=::NewEnumerator1
    Sprite Rotation Mode=::NewEnumerator2
    Sprite Size Mode=::NewEnumerator1
    Uniform Sprite Size Max=25
    Uniform Sprite Size Min=15
    UsePositionOffset=false
    Write Material Random=true
   )
   &ShapeLocation(
    MaskSpawnGroup=false
    Non Uniform Scale=&Lerp_Vector(A=&Lerp_Vector(A=<0.5 1 1>, B=<0.5 22 1>, Alpha=$User.OpenAmount), B=<0.5 1 1>, Alpha=$Particles.MaterialRandom)
    Offset=&MakeVector(X=0, Y=0, Z=&Lerp_Float(A=50, B=450, Alpha=$Particles.MaterialRandom))
    Offset Mode=::NewEnumerator0
    Override Seed=false
    Shape Origin=$Engine.Owner.Position
    Sphere Radius=10
    UseOwnerScale=true
   )
   &DynamicMaterialParameters(
    Index 0 Param 1=1
    Index 0 Param 2=1
    Index 0 Param 3=0.5
    Module.Param0WriteEnabled=true
    Module.Param1WriteEnabled=true
    Module.Param2WriteEnabled=true
    Module.Param3WriteEnabled=true
   )
   &AddVelocity(
    Additional Quaternion Rotation=false
    Rotation Mode=::NewEnumerator3
    Rotation Quaternion:Quat4f=$Engine.Owner.Rotation
    Velocity=&MakeVector(X=&Lerp_Float(A=-200, B=-125, Alpha=$Particles.MaterialRandom), Y=0, Z=0)
   )
  ]
  particleUpdateScript [
   &ParticleState()
   &ScaleColor(
    Scale Alpha=&FloatFromCurve(
     FloatCurve=DataInterface() {propertyValues {Curve {keys [
         <at 0 0 RCIM_Cubic>
         {interpMode=RCIM_Cubic, tangentMode=RCTM_User, time=0.0856, value=0.9914}
         {arriveTangent=-2.3503, interpMode=RCIM_Cubic, leaveTangent=-2.3503, tangentMode=RCTM_User, time=0.8081, value=0.5934}
         <at 1 0>
        ]}}}
     Scale Curve=&Lerp_Float(A=0.25, B=1, Alpha=$System.DistanceFade)
    )
    Scale RGB=&VectorFromFloat(Value=&Lerp_Float(A=1, B=2, Alpha=$User.WeldProgress))
    ScaleA=true
    ScaleRGB=true
   )
   &ScaleSpriteSize(Scale Factor=&Vector2DFromFloat(Value=&FloatFromCurve(
      FloatCurve=DataInterface() {propertyValues {Curve {keys [
          <at 0 0 RCIM_Cubic>
          {arriveTangent=0.836, interpMode=RCIM_Cubic, leaveTangent=0.836, time=0.2, value=1}
          {arriveTangent=-1.8455, interpMode=RCIM_Cubic, leaveTangent=-1.8455, tangentMode=RCTM_User, time=0.6507, value=0.544}
          <at 1 0>
         ]}}}
      Scale Curve=&Lerp_Float(A=0.5, B=1, Alpha=$System.OpenAmount)
     )))
   &VectorNoiseForce(Force Amount=800, Module.Use Random Noise Time Interval=true, Random Noise Time Interval=0.1)
   &GravityForce(Gravity=<0 0 -100>)
   &Drag(Use Linear Drag=true, Use Rotational Drag=false)
   &SolveForcesAndVelocity(Clamp Velocity=false, Limit Acceleration=false)
  ]
  rendererClasses [&NiagaraSpriteRendererProperties]
  renderers [{rendererClass=&NiagaraSpriteRendererProperties}]
 }
 StraightSplines {
  emitterUpdateScript [
   &EmitterState()
   &SpawnBurst_Instantaneous(Spawn Count=4, Use Spawn Probability=false, UseLoopCountLimit=true)
   &SpawnRate(SpawnRate=4, Use Spawn Probability=false)
  ]
  particleSpawnScript [
   &PartitionParticles()
   &InitializeParticle(Color=<rgba 0 2.5372 5 0.4>, Lifetime=4, Mesh Scale Mode=::NewEnumerator3, UsePositionOffset=false, Write Material Random=true)
   &CameraOffset(Camera Offset Amount=4)
   &InitialMeshOrientation(
    Mesh Orientation Mode=::NewEnumerator2
    Orientation Vector=<0 -1 0>
    Rotation=&Lerp_Vector(A=<0 0.04 0>, B=<0 -0.05 0>, Alpha=&MakeFloatFromInt(Int=$Particles.PartitionParticles.Partition))
    Use Rotation Vector=true
   )
  ]
  particleUpdateScript [
   &ParticleState()
   &ScaleMeshSize(Scale Factor=&MakeVector(X=$User.OpenAmount, Y=1, Z=1))
   &ScaleColor(
    Scale Alpha=&FloatFromCurve(
     FloatCurve=DataInterface() {propertyValues {Curve {keys [
         {arriveTangent=0.5021, interpMode=RCIM_Cubic, leaveTangent=0.5021, tangentMode=RCTM_User, time=0, value=0}
         {arriveTangent=0.0224, interpMode=RCIM_Cubic, leaveTangent=0.0224, tangentMode=RCTM_User, time=0.5, value=1}
         {arriveTangent=-5.2715, interpMode=RCIM_Cubic, leaveTangent=-5.2715, tangentMode=RCTM_User, time=1, value=0}
        ]}}}
     Scale Curve=$User.OpenAmount
    )
    Scale RGB=&MakeVectorFromLinearColorRGB(Color=&ColorFromCurve(ColorCurve=DataInterface() {propertyValues {
        AlphaCurve {keys [<at 0 1>]}
        BlueCurve {keys [
          <at 0.0288 5>
          {arriveTangent=-6.3639, interpMode=RCIM_Cubic, leaveTangent=-6.3639, time=0.142, value=2.5}
          {arriveTangent=-19.8906, leaveTangent=-19.8906, time=0.3523, value=1}
          <at 0.7972 0.1094 RCIM_Cubic>
         ]}
        GreenCurve {keys [
          <at 0.0288 3.6354>
          {arriveTangent=-4.6159, interpMode=RCIM_Cubic, leaveTangent=-4.6159, time=0.142, value=2.0541}
          {arriveTangent=-14.4534, leaveTangent=-14.4534, time=0.3523, value=0.6233}
          <at 0.7972 0.0881 RCIM_Cubic>
         ]}
        RedCurve {keys [
          <at 0.0288 2.4714>
          {arriveTangent=-3.1433, interpMode=RCIM_Cubic, leaveTangent=-3.1433, time=0.142, value=1.6691}
          {arriveTangent=-9.8298, leaveTangent=-9.8298, time=0.3523, value=0.3021}
          <at 0.7972 0.0558 RCIM_Cubic>
         ]}
       }}))
    ScaleA=true
    ScaleRGB=true
   )
   &DynamicMaterialParameters(
    Index 0 Param 1=&FloatFromCurve(FloatCurve=DataInterface() {propertyValues {Curve {keys [
         {interpMode=RCIM_Cubic, tangentMode=RCTM_User, time=0, value=1}
         {interpMode=RCIM_Cubic, tangentMode=RCTM_User, time=0.5, value=1}
         {arriveTangent=-2.7149, interpMode=RCIM_Cubic, leaveTangent=-2.7149, tangentMode=RCTM_User, time=1, value=0}
        ]}}})
    Index 0 Param 2=&FloatFromCurve(
     FloatCurve=DataInterface() {propertyValues {Curve {keys [
         {interpMode=RCIM_Cubic, tangentMode=RCTM_User, time=0, value=1}
         {interpMode=RCIM_Cubic, tangentMode=RCTM_User, time=0.5, value=1}
         {arriveTangent=-4.3107, interpMode=RCIM_Cubic, leaveTangent=-4.3107, tangentMode=RCTM_User, time=1, value=0}
        ]}}}
     Scale Curve=10
    )
    Index 0 Param 3=&FloatFromCurve(
     FloatCurve=DataInterface() {propertyValues {Curve {keys [<at 0 1>,<at 1 0>]}}}
     Scale Curve=&RandomRangeFloat(Minimum=1, Maximum=5)
    )
    Index 0 Param 4=&RandomRangeFloat(Minimum=0, Maximum=1)
    Module.Param0WriteEnabled=true
    Module.Param1WriteEnabled=true
    Module.Param2WriteEnabled=true
    Module.Param3WriteEnabled=true
   )
   /SetVariables(Particles.Position=&AddVectorToPosition(
     Position=$Engine.Emitter.SimulationPosition
     Vector=&Multiply_VectorByFloat(
      Vector=&Add_Vector(
       A=&Lerp_Vector(A=<0 120 0>, B=<0 -120 0>, Alpha=&MakeFloatFromInt(Int=$Particles.PartitionParticles.Partition))
       B=&Multiply_VectorByFloat(
        Vector=&Lerp_Vector(A=<0 180 0>, B=<0 -180 0>, Alpha=&MakeFloatFromInt(Int=$Particles.PartitionParticles.Partition))
        Float=&FloatFromCurve(
         FloatCurve=DataInterface() {propertyValues {Curve {keys [
             {interpMode=RCIM_Cubic, tangentMode=RCTM_User, time=0, value=0}
             {arriveTangent=0.2743, interpMode=RCIM_Cubic, leaveTangent=0.2743, tangentMode=RCTM_User, time=4, value=1}
            ]}}}
         CurveIndex=$Particles.Age
        )
       )
      )
      Float=$User.OpenAmount
     )
    )) {isSetParameters=true}
  ]
  rendererClasses [&NiagaraMeshRendererProperties]
  renderers [{rendererClass=&NiagaraMeshRendererProperties, rendererIndex=0}]
 }
}
summary {
 systemName=NS_PlotTwist_Portal_Opened
 emitters [
  {emitterName=Sparkles, rendererClasses [&NiagaraSpriteRendererProperties]}
  {emitterName=LightBeams, rendererClasses [&NiagaraSpriteRendererProperties]}
  {emitterName=Shimmers, rendererClasses [&NiagaraSpriteRendererProperties]}
  {emitterName=Motes, rendererClasses [&NiagaraSpriteRendererProperties]}
  {emitterName=Light, rendererClasses [&NiagaraSpriteRendererProperties, &NiagaraLightRendererProperties]}
  {emitterName=StraightSplines, rendererClasses [&NiagaraMeshRendererProperties]}
  {emitterName=GroundSprite, rendererClasses [&NiagaraSpriteRendererProperties]}
 ]
 userVariables [
  User.InteriorColor {type=LinearColor, description="", defaultValue {struct=LinearColor, value=<rgba 0.4452 0.4852 0.5149 1>}}
  User.MasterOpacity {type=NiagaraFloat, defaultValue=1, description=""}
  User.OpenAmount {type=NiagaraFloat, defaultValue=1, description=""}
  User.WeldProgress {type=NiagaraFloat, defaultValue=0, description=""}
 ]
}
