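# Post-process material for the Rift Gate item that wipes the screen to black in a clock-style radial sweep.
# It builds centered screen coordinates, rotates them, and combines a radial-angle mask with soft sphere-mask edges shaped by EdgeRadius and EdgeHardness.
# The Wipe scalar is exposed as the transition value.
# InvertWipe flips the mask and blends in a scene color capped by a circular mask.
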Material /PlotTwist/Items/RiftGate/Assets/Materials/M_Clockwipe_PP
graph {
 properties {materialDomain=MD_PostProcess, pixelDepthOffsetMode=PDOM_Legacy, shadingModel=MSM_Unlit, usedShadingModels=Unlit}
 nodes [
  n1 = ComponentMask(None=SceneTexture() {sceneTextureId=PPI_PostProcessInput0}.Color) {b=true, g=true, r=true}
  n2 = Constant() {r=0.6}
  n3 = SphereMask(A=NamedRerouteUsage(), B=@n2, Radius=@n2, Hardness=Constant() {r=0.9})
  n4 = ScalarParameter() {parameterName=InvertWipe}
  n5 = NamedRerouteUsage()
  n6 = Add(A=ConstantBiasScale(None=ScreenPosition().ViewportUV) {bias=-0.5, scale=2}, B=AppendVector(A=Multiply(A=@n5) {constB=-0.5}, B=Constant()))
  n7 = MaterialFunctionCall(Float2=Rotator(Coordinate=NamedRerouteDeclaration(None=@n6) {name=SignedScreenPosition, nodeColor=<rgba 1 0.1172 0 1>}, Time=NamedRerouteUsage()) {centerX=0, centerY=0, speed=-6.2832}) {materialFunction=&BreakOutFloat2Components}
  n8 = ScalarParameter() {defaultValue=0.09, parameterName=EdgeRadius}
  n9 = ScalarParameter() {defaultValue=0.1, parameterName=EdgeHardness}
  n10 = Sign(None=@n7.G)
  n11 = Multiply(
   A=SphereMask(
    A=@n7.R
    B=Constant()
    Radius=NamedRerouteDeclaration(None=@n8) {name=EdgeRadius, nodeColor=<rgba 1 0.7031 0 1>}
    Hardness=NamedRerouteDeclaration(None=@n9) {name=EdgeHardness, nodeColor=<rgba 0.5469 1 0 1>}
   )
   B=OneMinus(None=Saturate(None=@n10))
  )
  n12 = MaterialFunctionCall(Vector or UVs=Rotator(Coordinate=NamedRerouteUsage(), Time=Constant() {r=0.25}) {centerX=0, centerY=0, speed=6.2832}) {materialFunction=&VectorToRadialValue}
  n13 = MaterialFunctionCall(Float2=NamedRerouteUsage()) {materialFunction=&BreakOutFloat2Components}
  n14 = Constant()
  n15 = NamedRerouteUsage()
  n16 = NamedRerouteUsage()
  n17 = NamedRerouteUsage()
  n18 = Saturate(None=Sign(None=ComponentMask(None=@n17) {g=true}))
  n19 = Max(
   A=SphereMask(A=ComponentMask(None=@n12.Radial Coordinates) {r=true}, B=Constant(), Radius=NamedRerouteUsage())
   B=Multiply(A=SphereMask(A=@n13.R, B=@n14, Radius=@n15, Hardness=@n16), B=OneMinus(None=@n18))
  )
  n20 = NamedRerouteUsage()
  n21 = Constant()
  n22 = NamedRerouteUsage()
  n23 = NamedRerouteUsage()
  n24 = Saturate(None=Max(A=Max(A=@n11, B=@n19), B=SphereMask(A=@n20, B=@n21, Radius=@n22, Hardness=@n23)))
  LinearInterpolate(A=LinearInterpolate(A=@n1, B=Min(A=@n1, B=@n3), Alpha=@n4), Alpha=LinearInterpolate(A=@n24, B=OneMinus(None=@n24), Alpha=@n4)) {constB=0}
  MaterialFunctionCall!(Time, Divisor=Constant() {r=5}, Mod) {materialFunction=/Game/Effects/Fort_Effects/Materials/_Functions/MF_ModuloTime}
  NamedRerouteDeclaration(None=ScalarParameter() {parameterName=Wipe}) {name=TransitionScalar, nodeColor=<rgba 1 0.8203 0 1>}
 ]
}
