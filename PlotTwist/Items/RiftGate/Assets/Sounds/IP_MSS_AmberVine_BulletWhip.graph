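# One-shot stereo MetaSound that plays an AmberVine bullet-whip near-miss sound, judging by its names and attenuation settings.
# On play, two chained trigger compares gate the attenuation distance to a close band before a random-sound subgraph plays the whip-by wave array.
# A failed compare or finished playback fires OnFinished.
# A concurrency override allows one instance and prevents new ones; the Speed and orientation inputs are unused.
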MetaSoundSource /PlotTwist/Items/RiftGate/Assets/Sounds/IP_MSS_AmberVine_BulletWhip
inputs [
 3P Bys {type=WaveAsset:Array}
 Speed {type=Float}
 UE.Attenuation.Distance {type=Float}
 UE.Source.OnPlay {type=Trigger}
 UE.Source.Orientation.Azimuth {type=Float}
 UE.Source.Orientation.Elevation {type=Float}
]
interfaces [UE.Attenuation:1.0, UE.OutputFormat.Stereo:1.0, UE.Source.OneShot:1.0, UE.Source.Orientation:1.0, UE.Source:1.1]
outputs [
 "UE.OutputFormat.Stereo.Audio:0" {type=Audio}
 "UE.OutputFormat.Stereo.Audio:1" {type=Audio}
 UE.Source.OneShot.OnFinished {type=Trigger}
]
pages [Default {
  input_defaults {
   Speed=0.856
   UE.Attenuation.Distance=536.005
   UE.Source.OnPlay=null
   UE.Source.Orientation.Azimuth=0
   UE.Source.Orientation.Elevation=0
   3P Bys [/PlotTwist/Items/RiftGate/Assets/Sounds/Wavs/IP_Ambervine_NearMiss_WhipBy_01]
  }
  nodes [
   InitVariable.Float()
   n1 = TriggerCompare.Clamp:Float(A=^UE.Attenuation.Distance, Compare=^UE.Source.OnPlay) {also {defaults {B=550, Type=2}}}
   n2 = TriggerCompare.Clamp:Float(A=^UE.Attenuation.Distance, Compare=@n1.True) {also {defaults {B=50, Type=3}}}
   MSS_PlayRandomSound_Stereo_0 = &MSS_PlayRandomSound_Stereo(Sounds=^3P Bys, UE.Source.OnPlay=@n2.True)
   TriggerAnyTriggerAny3_0 = "TriggerAny.Trigger Any (3)"(In 2=@MSS_PlayRandomSound_Stereo_0.UE.Source.OneShot.OnFinished, In 1=@n1.False, In 0=@n2.False)
  ]
  links [
   {dst="@UE.OutputFormat.Stereo.Audio:0", src=MSS_PlayRandomSound_Stereo_0.UE.OutputFormat.Stereo.Audio:0}
   {dst="@UE.OutputFormat.Stereo.Audio:1", src=MSS_PlayRandomSound_Stereo_0.UE.OutputFormat.Stereo.Audio:1}
   {dst="@UE.Source.OneShot.OnFinished", src=TriggerAnyTriggerAny3_0.Out}
  ]
 }]
properties {
 assetImportData=&self:AssetImportData
 attenuationSettings=/PlotTwist/Items/RiftGate/Assets/Sounds/Settings/ATT_AmberVine_NearMiss
 bOverrideConcurrency=true
 maxDistance=4500
 numChannels=2
 outputFormat=Stereo
 pitch=1.2
 sampleRate=48000
 soundClassObject=&Fort_Gliders_InAir
 volume=0.3
 concurrencyOverrides {maxCount=1, resolutionRule=PreventNew, retriggerTime=2}
}
