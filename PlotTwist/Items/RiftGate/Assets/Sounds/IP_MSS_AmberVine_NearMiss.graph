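# Mono MetaSound for the Rift Gate AmberVine near miss, playing a whip-by from a wave array as attenuation distance closes in.
# A falling distance threshold fires through a Trigger Control gate and a Speed comparison into a random array pick and Wave Player.
# Playback start closes the gate via a deferred trigger variable, and a larger rising threshold reopens it, giving hysteresis against repeats.
# The looping source allows one concurrent instance and prevents new ones.
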MetaSoundSource /PlotTwist/Items/RiftGate/Assets/Sounds/IP_MSS_AmberVine_NearMiss
inputs [
 3P Bys {type=WaveAsset:Array}
 Speed {type=Float}
 UE.Attenuation.Distance {type=Float}
 UE.Source.OnPlay {type=Trigger}
 UE.Source.Orientation.Azimuth {type=Float}
 UE.Source.Orientation.Elevation {type=Float}
]
interfaces [UE.Attenuation:1.0, UE.OutputFormat.Mono:1.0, UE.Source.Orientation:1.0, UE.Source:1.1]
outputs ["UE.OutputFormat.Mono.Audio:0" {type=Audio}]
pages [Default {
  input_defaults {
   Speed=0.856
   UE.Attenuation.Distance=0
   UE.Source.OnPlay=null
   UE.Source.Orientation.Azimuth=0
   UE.Source.Orientation.Elevation=0
   3P Bys [/PlotTwist/Items/RiftGate/Assets/Sounds/Wavs/IP_Ambervine_NearMiss_WhipBy_01]
  }
  nodes [
   InitVariable.Float()
   n1 = VariableDeferredAccessor.Trigger(Variable=InitVariable.Trigger().Variable)
   TriggerControl_0 = UE.Trigger Control(
    Trigger In=UE.TriggerOnThreshold:Float(In=^UE.Attenuation.Distance) {also {defaults {Threshold=800, Type=1}}}.Out
    Open=UE.TriggerOnThreshold:Float(In=^UE.Attenuation.Distance) {also {defaults {Threshold=1000}}}.Out
    Close=@n1.Value
   )
   ArrayRandomGet_0 = Array.Random Get:WaveAsset:Array(In Array=^3P Bys, Next=TriggerCompare.Clamp:Float(A=^Speed, Compare=@TriggerControl_0.Trigger Out) {also {defaults {Type=3}}}.True)
   WavePlayer_0 = UE.Wave Player:Mono(Play=@ArrayRandomGet_0.On Next, Wave Asset=@ArrayRandomGet_0.Value)
   VariableMutator.Trigger(Variable=@n1.Variable, Value=@WavePlayer_0.On Play)
  ]
  links [{dst="@UE.OutputFormat.Mono.Audio:0", src=WavePlayer_0.Out Mono}]
 }]
properties {
 assetImportData=&self:AssetImportData
 attenuationSettings=/PlotTwist/Items/RiftGate/Assets/Sounds/Settings/ATT_AmberVine_NearMiss
 bLooping=true
 bOverrideConcurrency=true
 duration=10000
 maxDistance=4500
 pitch=1.2
 sampleRate=48000
 soundClassObject=&Fort_Gliders_InAir
 concurrencyOverrides {maxCount=1, resolutionRule=PreventNew, retriggerTime=5}
}
