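# Looping stereo MetaSound source for the Ambervine EMP loop in the RiftGate item's sounds, with its own attenuation.
# On play, Main and Sweetener layers each feed a random looping player, with volume and pitch ranges built from base plus randomization offsets.
# A two-input mixer feeds an affiliation patch that sets level and lowpass by local, friendly or enemy status.
# By default, Main holds one wave and Sweetener is empty.
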MetaSoundSource /PlotTwist/Items/RiftGate/Assets/Sounds/IP_MSS_Ambervine_EMP_Lp
inputs [
 Affiliation.Enemy.Db {type=Float}
 Affiliation.Enemy.LP {type=Float}
 Affiliation.Friendly.Db {type=Float}
 Affiliation.Friendly.LP {type=Float}
 Affiliation.Local.Db {type=Float}
 Character.Affiliation {type=Enum:FortAudioAffiliation}
 Character.Health {type=Float}
 Character.UniqueID {type=Int32}
 Main.Db {type=Float}
 Main.Randomization.DbMax {type=Float}
 Main.Randomization.DbMin {type=Float}
 Main.Randomization.SemitoneMax {type=Float}
 Main.Randomization.SemitoneMin {type=Float}
 Main.Semitone {type=Float}
 Main.Waves {type=WaveAsset:Array}
 Sweetener.Db {type=Float}
 Sweetener.Randomize.DbMax {type=Float}
 Sweetener.Randomize.DbMin {type=Float}
 Sweetener.Randomize.SemitoneMax {type=Float}
 Sweetener.Randomize.SemitoneMin {type=Float}
 Sweetener.Semitone {type=Float}
 Sweetener.Waves {type=WaveAsset:Array}
 UE.Source.OnPlay {type=Trigger}
]
interfaces [Fort.Character:1.0, UE.OutputFormat.Stereo:1.0, UE.Source:1.1]
outputs [
 "UE.OutputFormat.Stereo.Audio:0" {type=Audio}
 "UE.OutputFormat.Stereo.Audio:1" {type=Audio}
]
pages [Default {
  input_defaults {
   Affiliation.Enemy.Db=0
   Affiliation.Enemy.LP=20000
   Affiliation.Friendly.Db=-4
   Affiliation.Friendly.LP=20000
   Affiliation.Local.Db=0
   Character.Affiliation=0
   Character.Health=0
   Character.UniqueID=0
   Main.Db=-9.5065
   Main.Randomization.DbMax=2
   Main.Randomization.DbMin=-2
   Main.Randomization.SemitoneMax=-1
   Main.Randomization.SemitoneMin=0
   Main.Semitone=0
   Sweetener.Db=-12
   Sweetener.Randomize.DbMax=2
   Sweetener.Randomize.DbMin=-2
   Sweetener.Randomize.SemitoneMax=1
   Sweetener.Randomize.SemitoneMin=-1
   Sweetener.Semitone=0
   UE.Source.OnPlay=null
   Main.Waves [/PlotTwist/Items/RiftGate/Assets/Sounds/Wavs/IP_Ambervine_EMP_Lp_01]
   Sweetener.Waves []
  }
  nodes [
   DecibelstoLinearGain_2 = UE.Decibels to Linear Gain:Float(Decibels=UE.Add:Float(PrimaryOperand=^Main.Db, AdditionalOperands=^Main.Randomization.DbMin).Out)
   DecibelstoLinearGain_3 = UE.Decibels to Linear Gain:Float(Decibels=UE.Add:Float(PrimaryOperand=^Main.Db, AdditionalOperands=^Main.Randomization.DbMax).Out)
   SemitonetoFrequencyMultiplier_1 = UE.Semitone to Frequency Multiplier:Float(Semitones=UE.Add:Float(AdditionalOperands=^Main.Randomization.SemitoneMin, PrimaryOperand=^Main.Semitone).Out)
   SemitonetoFrequencyMultiplier_3 = UE.Semitone to Frequency Multiplier:Float(Semitones=UE.Add:Float(AdditionalOperands=^Main.Randomization.SemitoneMax, PrimaryOperand=^Main.Semitone).Out)
   n1 = /FortMetasound/Sources/MSS_PlayRandomSoundLooping_Stereo(
    Sounds=^Main.Waves
    UE.Source.OnPlay=^UE.Source.OnPlay
    Min Volume=@DecibelstoLinearGain_2.Linear Gain
    Max Volume=@DecibelstoLinearGain_3.Linear Gain
    Min Pitch=@SemitonetoFrequencyMultiplier_1.Frequency Multiplier
    Max Pitch=@SemitonetoFrequencyMultiplier_3.Frequency Multiplier
   )
   DecibelstoLinearGain_0 = UE.Decibels to Linear Gain:Float(Decibels=UE.Add:Float(PrimaryOperand=^Sweetener.Db, AdditionalOperands=^Sweetener.Randomize.DbMax).Out)
   DecibelstoLinearGain_1 = UE.Decibels to Linear Gain:Float(Decibels=UE.Add:Float(PrimaryOperand=^Sweetener.Db, AdditionalOperands=^Sweetener.Randomize.DbMin).Out)
   SemitonetoFrequencyMultiplier_0 = UE.Semitone to Frequency Multiplier:Float(Semitones=UE.Add:Float(AdditionalOperands=^Sweetener.Randomize.SemitoneMin, PrimaryOperand=^Sweetener.Semitone).Out)
   SemitonetoFrequencyMultiplier_2 = UE.Semitone to Frequency Multiplier:Float(Semitones=UE.Add:Float(AdditionalOperands=^Sweetener.Randomize.SemitoneMax, PrimaryOperand=^Sweetener.Semitone).Out)
   n2 = &MSS_PlayRandomSoundLooping_Stereo(
    Sounds=^Sweetener.Waves
    UE.Source.OnPlay=^UE.Source.OnPlay
    Max Volume=@DecibelstoLinearGain_0.Linear Gain
    Min Volume=@DecibelstoLinearGain_1.Linear Gain
    Min Pitch=@SemitonetoFrequencyMultiplier_0.Frequency Multiplier
    Max Pitch=@SemitonetoFrequencyMultiplier_2.Frequency Multiplier
   )
   AudioMixerAudioMixerStereo2_0 = "AudioMixer.Audio Mixer (Stereo, 2)"(
    In 0 L=@n1."UE.OutputFormat.Stereo.Audio:0"
    In 0 R=@n1."UE.OutputFormat.Stereo.Audio:1"
    In 1 L=@n2."UE.OutputFormat.Stereo.Audio:0"
    In 1 R=@n2."UE.OutputFormat.Stereo.Audio:1"
   )
   MSP_Audio_AffiliationMixer_Stereo_0 = /FortMetasound/Patches/MSP_Audio_AffiliationMixer_Stereo(
    Enemy.DbAdjustment=^Affiliation.Enemy.Db
    Enemy.Lowpass=^Affiliation.Enemy.LP
    Friendly.DbAdjustment=^Affiliation.Friendly.Db
    Friendly.Lowpass=^Affiliation.Friendly.LP
    Local.DbAdjustment=^Affiliation.Local.Db
    OnPlay=^UE.Source.OnPlay
    In L=@AudioMixerAudioMixerStereo2_0.Out L
    In R=@AudioMixerAudioMixerStereo2_0.Out R
    Affiliation=Convert.Int32:Enum:FortAudioAffiliation("Enum:FortAudioAffiliation"=^Character.Affiliation).Int32
   )
  ]
  links [
   {dst="@UE.OutputFormat.Stereo.Audio:0", src=MSP_Audio_AffiliationMixer_Stereo_0.Out L}
   {dst="@UE.OutputFormat.Stereo.Audio:1", src=MSP_Audio_AffiliationMixer_Stereo_0.Out R}
  ]
 }]
properties {
 assetImportData=&self:AssetImportData
 attenuationSettings=/PlotTwist/Items/RiftGate/Assets/Sounds/Settings/ATT_AmberVine_EMP
 bLooping=true
 duration=10000
 maxDistance=4400
 numChannels=2
 outputFormat=Stereo
 sampleRate=48000
 soundClassObject=&Fort_WeaponMelee
}
