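# Looping stereo MetaSound for the Ambervine lightning beam of the RiftGate item, scaled by character affiliation.
# Randomized main and sweetener loops each pass through an EQ crossfade driven by interpolated PctElapsed.
# The larger of normalized absolute Yaw and Pitch is smoothed and drives a reused glider direction patch playing the sideways loop and whip-by one-shots.
# PctElapsed-based semitone offsets are computed but never reach the players.
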MetaSoundSource /PlotTwist/Items/RiftGate/Assets/Sounds/IP_MSS_Ambervine_Lightning_Lp
inputs [
 Character.Affiliation {type=Enum:FortAudioAffiliation}
 Character.Health {type=Float}
 Character.UniqueID {type=Int32}
 InterpTime {type=Float}
 Main.Db {type=Float}
 Main.Randomization.DbSpread {type=Float}
 Main.Randomization.SemitoneSpread {type=Float}
 Main.Semitone {type=Float}
 Main.Waves {type=WaveAsset:Array}
 PctElapsed {type=Float}
 Pitch {type=Float}
 Sideways.Db {type=Float}
 Sideways.Loop {type=WaveAsset:Array}
 Sideways.OneShot {type=WaveAsset:Array}
 Sweetener.Db {type=Float}
 Sweetener.Randomize.DbSpread {type=Float}
 Sweetener.Randomize.SemitoneSpread {type=Float}
 Sweetener.Semitone {type=Float}
 Sweetener.Waves {type=WaveAsset:Array}
 UE.Source.OnPlay {type=Trigger}
 Yaw {type=Float}
]
interfaces [Fort.Character:1.0, UE.OutputFormat.Stereo:1.0, UE.Source:1.1]
outputs [
 "UE.OutputFormat.Stereo.Audio:0" {type=Audio}
 "UE.OutputFormat.Stereo.Audio:1" {type=Audio}
]
pages [Default {
  input_defaults {
   Character.Affiliation=0
   Character.Health=0
   Character.UniqueID=0
   InterpTime=0
   Main.Db=-10
   Main.Randomization.DbSpread=2
   Main.Randomization.SemitoneSpread=1
   Main.Semitone=0
   PctElapsed=0
   Pitch=0
   Sideways.Db=0
   Sweetener.Db=-11
   Sweetener.Randomize.DbSpread=2
   Sweetener.Randomize.SemitoneSpread=1
   Sweetener.Semitone=0
   UE.Source.OnPlay=null
   Yaw=0
   Main.Waves [/PlotTwist/Items/RiftGate/Assets/Sounds/Wavs/IP_AmberVine_Lightning_Lp_02]
   Sideways.Loop [/PlotTwist/Items/RiftGate/Assets/Sounds/Wavs/IP_AmberVine_Lightning_Lp_Sideways_02]
   Sideways.OneShot [
    /PlotTwist/Items/RiftGate/Assets/Sounds/Wavs/IP_Ambervine_NearMiss_WhipBy_04
    /PlotTwist/Items/RiftGate/Assets/Sounds/Wavs/IP_Ambervine_NearMiss_WhipBy_03
    /PlotTwist/Items/RiftGate/Assets/Sounds/Wavs/IP_Ambervine_NearMiss_WhipBy_02
    /PlotTwist/Items/RiftGate/Assets/Sounds/Wavs/IP_Ambervine_NearMiss_WhipBy_01
   ]
   Sweetener.Waves [/PlotTwist/Items/RiftGate/Assets/Sounds/Wavs/IP_AmberVine_Lightning_Lp_Swt_02]
  }
  nodes [
   n1 = InitVariable.Float()
   n2 = UE.ConversionFloatToTime(In=^InterpTime)
   n3 = VariableAccessor.Float(Variable=VariableMutator.Float(Variable=InitVariable.Float().Variable, Value=UE.InterpTo:Audio(Target=^PctElapsed, Interp Time=@n2.Out).Value).Variable)
   n4 = VariableAccessor.Float(Variable=@n3.Variable)
   n5 = MapRange.MapRange:Float(In=@n4.Value) {also {defaults {Out Range B=16}}}
   UE.Add:Float(
    PrimaryOperand=^Sweetener.Semitone
    AdditionalOperands=VariableAccessor.Float(Variable=VariableMutator.Float(Variable=@n1.Variable, Value=@n5.Out Value).Variable).Value
   )
   n6 = InitVariable.Float()
   n7 = MapRange.MapRange:Float(In=@n3.Value) {also {defaults {Out Range B=16}}}
   UE.Add:Float(
    PrimaryOperand=^Main.Semitone
    AdditionalOperands=VariableAccessor.Float(Variable=VariableMutator.Float(Variable=@n6.Variable, Value=@n7.Out Value).Variable).Value
   )
   n8 = /Metasound/DefaultAssets/Sources/MSS_PlayRandom_Loop(VolumeBase=^Main.Db, VolumeRandomSpread=^Main.Randomization.DbSpread, Sounds=^Main.Waves, UE.Source.OnPlay=^UE.Source.OnPlay)
   n9 = VariableAccessor.Float(Variable=@n4.Variable)
   n10 = /FortMetasound/Patches/MSP_Audio_EQCrossFade(In L=@n8."UE.OutputFormat.Stereo.Audio:0", In R=@n8."UE.OutputFormat.Stereo.Audio:1", Input=@n9.Value)
   n11 = InitVariable.Float()
   n12 = &MSS_PlayRandom_Loop(
    VolumeRandomSpread=^Sweetener.Randomize.DbSpread
    Sounds=^Sweetener.Waves
    UE.Source.OnPlay=^UE.Source.OnPlay
    VolumeBase=UE.Add:Float(PrimaryOperand=^Sweetener.Db, AdditionalOperands=VariableAccessor.Float(Variable=@n11.Variable).Value).Out
   )
   n13 = &MSP_Audio_EQCrossFade(
    In L=@n12."UE.OutputFormat.Stereo.Audio:0"
    In R=@n12."UE.OutputFormat.Stereo.Audio:1"
    Input=VariableAccessor.Float(Variable=@n9.Variable).Value
   )
   AudioMixerAudioMixerStereo2_1 = "AudioMixer.Audio Mixer (Stereo, 2)"(In 0 L=@n10.Out L, In 0 R=@n10.Out R, In 1 L=@n13.Out L, In 1 R=@n13.Out R)
   n14 = InitVariable.Float()
   n15 = MapRange.MapRange:Float(In=AbsoluteValue.Abs:Float(Input=^Yaw).Value) {also {defaults {In Range B=5}}}
   n16 = MapRange.MapRange:Float(In=AbsoluteValue.Abs:Float(Input=^Pitch).Value) {also {defaults {In Range B=250}}}
   TriggerOnValueChange_0 = UE.Trigger On Value Change:Float(Value=^Yaw)
   TriggerOnValueChange_1 = UE.Trigger On Value Change:Float(Value=^Pitch)
   TriggerAnyTriggerAny2_0 = "TriggerAny.Trigger Any (2)"(In 1=@TriggerOnValueChange_0.Trigger, In 0=@TriggerOnValueChange_1.Trigger)
   n17 = TriggerCompare.Clamp:Float(B=@n15.Out Value, A=@n16.Out Value, Compare=@TriggerAnyTriggerAny2_0.Out) {also {defaults {Type=3}}}
   TriggerRouteTriggerRouteFloat2_0 = "TriggerRoute.Trigger Route (Float, 2):Float"(Value 1=@n15.Out Value, Value 0=@n16.Out Value, Set 1=@n17.False, Set 0=@n17.True)
   n18 = &InterpSpeed(RisingInterpSpeed=^InterpTime, FloatInput=@TriggerRouteTriggerRouteFloat2_0.Value) {also {defaults {FallingInterpSpeed=0.6}}}
   n19 = /FortMetasound/Cosmetics/Glider/Glider_ThrustLoop_Direction(
    Loop=^Sideways.Loop
    Oneshots=^Sideways.OneShot
    On Play=^UE.Source.OnPlay
    Value=VariableAccessor.Float(Variable=VariableMutator.Float(Variable=@n14.Variable, Value=@n18.FloatOutput).Variable).Value
   ) {also {defaults {Interp Time=0.1, OneShotVolume=0.4, PitchMax=1.1, VolumeMax=0.75}}}
   n20 = &MSP_AffiliationMixFloat(Affiliation=^Character.Affiliation) {also {defaults {Non-Enemy=0.5}}}
   AudioMixerAudioMixerStereo2_0 = "AudioMixer.Audio Mixer (Stereo, 2)"(
    In 0 L=@AudioMixerAudioMixerStereo2_1.Out L
    In 0 R=@AudioMixerAudioMixerStereo2_1.Out R
    In 1 L=@n19.Out L
    In 1 R=@n19.Out R
    Gain 0=@n20.Out
    Gain 1=@n20.Out
   )
   InitVariable.Float()
  ]
  links [
   {dst="@UE.OutputFormat.Stereo.Audio:0", src=AudioMixerAudioMixerStereo2_0.Out L}
   {dst="@UE.OutputFormat.Stereo.Audio:1", src=AudioMixerAudioMixerStereo2_0.Out R}
  ]
 }]
properties {
 assetImportData=&self:AssetImportData
 attenuationSettings=/PlotTwist/Items/RiftGate/Assets/Sounds/Settings/ATT_AmberVine_Beam
 bLooping=true
 duration=10000
 maxDistance=8000
 numChannels=2
 outputFormat=Stereo
 sampleRate=48000
 soundClassObject=&Fort_WeaponMelee
 volume=2
}
