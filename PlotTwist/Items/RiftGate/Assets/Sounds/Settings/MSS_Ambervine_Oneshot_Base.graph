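# Stereo one-shot MetaSound base for Ambervine RiftGate item sounds, with empty wave arrays and the melee weapon sound class.
# Main and Sweetener layers play random waves with decibel and semitone controls and randomization, then mix together.
# An affiliation mixer applies local, friendly or enemy levels, plus lowpass for the latter two.
# It receives OnPlay immediately while both layers start briefly delayed, and OnFinished waits for both layers.
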MetaSoundSource /PlotTwist/Items/RiftGate/Assets/Sounds/Settings/MSS_Ambervine_Oneshot_Base
inputs [
 Affiliation.Enemy.Db {type=Float}
 Affiliation.Enemy.LP {type=Float}
 Affiliation.Friendly.Db {type=Float}
 Affiliation.Friendly.LP {type=Float}
 Affiliation.Local.Db {type=Float}
 Character.Affiliation {type=Enum:FortAudioAffiliation}
 Character.Health {type=Float}
 Character.UniqueID {type=Int32}
 Main.Db {type=Float}
 Main.Randomization.DbMax {type=Float}
 Main.Randomization.DbMin {type=Float}
 Main.Randomization.SemitoneMax {type=Float}
 Main.Randomization.SemitoneMin {type=Float}
 Main.Semitone {type=Float}
 Main.Waves {type=WaveAsset:Array}
 Sweetener.Db {type=Float}
 Sweetener.Randomize.DbMax {type=Float}
 Sweetener.Randomize.DbMin {type=Float}
 Sweetener.Randomize.SemitoneMax {type=Float}
 Sweetener.Randomize.SemitoneMin {type=Float}
 Sweetener.Semitone {type=Float}
 Sweetener.Waves {type=WaveAsset:Array}
 UE.Source.OnPlay {type=Trigger}
]
interfaces [Fort.Character:1.0, UE.OutputFormat.Stereo:1.0, UE.Source.OneShot:1.0, UE.Source:1.1]
outputs [
 "UE.OutputFormat.Stereo.Audio:0" {type=Audio}
 "UE.OutputFormat.Stereo.Audio:1" {type=Audio}
 UE.Source.OneShot.OnFinished {type=Trigger}
]
pages [Default {
  input_defaults {
   Affiliation.Enemy.Db=0
   Affiliation.Enemy.LP=20000
   Affiliation.Friendly.Db=-4
   Affiliation.Friendly.LP=20000
   Affiliation.Local.Db=0
   Character.Affiliation=0
   Character.Health=0
   Character.UniqueID=0
   Main.Db=0
   Main.Randomization.DbMax=2
   Main.Randomization.DbMin=-2
   Main.Randomization.SemitoneMax=-1
   Main.Randomization.SemitoneMin=0
   Main.Semitone=0
   Sweetener.Db=0
   Sweetener.Randomize.DbMax=2
   Sweetener.Randomize.DbMin=-2
   Sweetener.Randomize.SemitoneMax=1
   Sweetener.Randomize.SemitoneMin=-1
   Sweetener.Semitone=0
   UE.Source.OnPlay=null
   Main.Waves []
   Sweetener.Waves []
  }
  nodes [
   InitVariable.Trigger()
   DecibelstoLinearGain_0 = UE.Decibels to Linear Gain:Float(Decibels=^Sweetener.Randomize.DbMax)
   DecibelstoLinearGain_1 = UE.Decibels to Linear Gain:Float(Decibels=^Sweetener.Randomize.DbMin)
   DecibelstoLinearGain_3 = UE.Decibels to Linear Gain:Float(Decibels=^Sweetener.Db)
   SemitonetoFrequencyMultiplier_2 = UE.Semitone to Frequency Multiplier:Float(Semitones=^Sweetener.Randomize.SemitoneMin)
   SemitonetoFrequencyMultiplier_4 = UE.Semitone to Frequency Multiplier:Float(Semitones=^Sweetener.Randomize.SemitoneMax)
   SemitonetoFrequencyMultiplier_5 = UE.Semitone to Frequency Multiplier:Float(Semitones=^Sweetener.Semitone)
   TriggerDelay_0 = UE.Trigger Delay(In=^UE.Source.OnPlay) {also {defaults {Delay Time=0.01}}}
   n1 = InitVariable.Trigger()
   n2 = VariableAccessor.Trigger(Variable=VariableMutator.Trigger(Value=@TriggerDelay_0.Out, Variable=VariableDeferredAccessor.Trigger(Variable=@n1.Variable).Variable).Variable)
   n3 = &MSS_PlayRandomSound_Stereo(
    Sounds=^Sweetener.Waves
    MaxVolume=@DecibelstoLinearGain_0.Linear Gain
    MinVolume=@DecibelstoLinearGain_1.Linear Gain
    Volume=@DecibelstoLinearGain_3.Linear Gain
    MinPitch=@SemitonetoFrequencyMultiplier_2.Frequency Multiplier
    MaxPitch=@SemitonetoFrequencyMultiplier_4.Frequency Multiplier
    Pitch=@SemitonetoFrequencyMultiplier_5.Frequency Multiplier
    UE.Source.OnPlay=@n2.Value
   )
   DecibelstoLinearGain_2 = UE.Decibels to Linear Gain:Float(Decibels=^Main.Randomization.DbMin)
   DecibelstoLinearGain_4 = UE.Decibels to Linear Gain:Float(Decibels=^Main.Randomization.DbMax)
   DecibelstoLinearGain_5 = UE.Decibels to Linear Gain:Float(Decibels=^Main.Db)
   SemitonetoFrequencyMultiplier_0 = UE.Semitone to Frequency Multiplier:Float(Semitones=^Main.Randomization.SemitoneMax)
   SemitonetoFrequencyMultiplier_1 = UE.Semitone to Frequency Multiplier:Float(Semitones=^Main.Semitone)
   SemitonetoFrequencyMultiplier_3 = UE.Semitone to Frequency Multiplier:Float(Semitones=^Main.Randomization.SemitoneMin)
   n4 = &MSS_PlayRandomSound_Stereo(
    Sounds=^Main.Waves
    MinVolume=@DecibelstoLinearGain_2.Linear Gain
    MaxVolume=@DecibelstoLinearGain_4.Linear Gain
    Volume=@DecibelstoLinearGain_5.Linear Gain
    MaxPitch=@SemitonetoFrequencyMultiplier_0.Frequency Multiplier
    Pitch=@SemitonetoFrequencyMultiplier_1.Frequency Multiplier
    MinPitch=@SemitonetoFrequencyMultiplier_3.Frequency Multiplier
    UE.Source.OnPlay=@n2.Value
   )
   AudioMixerAudioMixerStereo2_0 = "AudioMixer.Audio Mixer (Stereo, 2)"(
    In 1 L=@n3."UE.OutputFormat.Stereo.Audio:0"
    In 1 R=@n3."UE.OutputFormat.Stereo.Audio:1"
    In 0 L=@n4."UE.OutputFormat.Stereo.Audio:0"
    In 0 R=@n4."UE.OutputFormat.Stereo.Audio:1"
   )
   MSP_Audio_AffiliationMixer_Stereo_0 = /FortMetasound/Patches/MSP_Audio_AffiliationMixer_Stereo(
    Enemy.DbAdjustment=^Affiliation.Enemy.Db
    Enemy.Lowpass=^Affiliation.Enemy.LP
    Friendly.DbAdjustment=^Affiliation.Friendly.Db
    Friendly.Lowpass=^Affiliation.Friendly.LP
    Local.DbAdjustment=^Affiliation.Local.Db
    OnPlay=^UE.Source.OnPlay
    In L=@AudioMixerAudioMixerStereo2_0.Out L
    In R=@AudioMixerAudioMixerStereo2_0.Out R
    Affiliation=Convert.Int32:Enum:FortAudioAffiliation("Enum:FortAudioAffiliation"=^Character.Affiliation).Int32
   )
   TriggerAccumulatorTriggerAccumulate2_0 = "TriggerAccumulator.Trigger Accumulate (2)"(In 1=@n3.UE.Source.OneShot.OnFinished, In 0=@n4.UE.Source.OneShot.OnFinished)
  ]
  links [
   {dst="@UE.OutputFormat.Stereo.Audio:0", src=MSP_Audio_AffiliationMixer_Stereo_0.Out L}
   {dst="@UE.OutputFormat.Stereo.Audio:1", src=MSP_Audio_AffiliationMixer_Stereo_0.Out R}
   {dst="@UE.Source.OneShot.OnFinished", src=TriggerAccumulatorTriggerAccumulate2_0.Out}
  ]
 }]
properties {
 assetImportData=&self:AssetImportData
 numChannels=2
 outputFormat=Stereo
 sampleRate=48000
 soundClassObject=&Fort_WeaponMelee
}
