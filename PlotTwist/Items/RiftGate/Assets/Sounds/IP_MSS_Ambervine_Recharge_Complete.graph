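# MetaSound Source preset that plays the stereo one-shot for Ambervine (RiftGate item) recharge completion.
# It inherits its graph from MSS_Ambervine_Oneshot_Base, supplying a borrowed Storm King lightning-appear wave with randomized level and pitch, and no sweetener waves.
# Affiliation gains keep it at full level for the local player while making it near-silent for enemies and silent for friendlies.
# It uses the shared AmberVine attenuation and melee weapon sound class.
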MetaSoundSource /PlotTwist/Items/RiftGate/Assets/Sounds/IP_MSS_Ambervine_Recharge_Complete
preset_parent=/PlotTwist/Items/RiftGate/Assets/Sounds/Settings/MSS_Ambervine_Oneshot_Base
inputs [
 Affiliation.Enemy.Db {type=Float}
 Affiliation.Enemy.LP {type=Float}
 Affiliation.Friendly.Db {type=Float}
 Affiliation.Friendly.LP {type=Float}
 Affiliation.Local.Db {type=Float}
 Character.Affiliation {type=Enum:FortAudioAffiliation}
 Character.Health {type=Float}
 Character.UniqueID {type=Int32}
 Main.Db {type=Float}
 Main.Randomization.DbMax {type=Float}
 Main.Randomization.DbMin {type=Float}
 Main.Randomization.SemitoneMax {type=Float}
 Main.Randomization.SemitoneMin {type=Float}
 Main.Semitone {type=Float}
 Main.Waves {type=WaveAsset:Array}
 Sweetener.Db {type=Float}
 Sweetener.Randomize.DbMax {type=Float}
 Sweetener.Randomize.DbMin {type=Float}
 Sweetener.Randomize.SemitoneMax {type=Float}
 Sweetener.Randomize.SemitoneMin {type=Float}
 Sweetener.Semitone {type=Float}
 Sweetener.Waves {type=WaveAsset:Array}
 UE.Source.OnPlay {type=Trigger}
]
interfaces [Fort.Character:1.0, UE.OutputFormat.Stereo:1.0, UE.Source.OneShot:1.0, UE.Source:1.1]
outputs [
 "UE.OutputFormat.Stereo.Audio:0" {type=Audio}
 "UE.OutputFormat.Stereo.Audio:1" {type=Audio}
 UE.Source.OneShot.OnFinished {type=Trigger}
]
pages [Default {input_defaults {
   Affiliation.Enemy.Db=-60
   Affiliation.Enemy.LP=20000
   Affiliation.Friendly.Db=-120
   Affiliation.Friendly.LP=20000
   Affiliation.Local.Db=0
   Character.Affiliation=0
   Character.Health=0
   Character.UniqueID=0
   Main.Db=-18
   Main.Randomization.DbMax=2
   Main.Randomization.DbMin=-2
   Main.Randomization.SemitoneMax=-1
   Main.Randomization.SemitoneMin=0
   Main.Semitone=0
   Sweetener.Db=0
   Sweetener.Randomize.DbMax=2
   Sweetener.Randomize.DbMin=-2
   Sweetener.Randomize.SemitoneMax=1
   Sweetener.Randomize.SemitoneMin=-1
   Sweetener.Semitone=0
   UE.Source.OnPlay=null
   Main.Waves [/Game/Sounds/Fort_Monster_Storm_King/SFX/Final_SFX/StormKing_Projectile_Lightning_Appear_03]
   Sweetener.Waves []
  }}]
properties {
 assetImportData=&self:AssetImportData
 attenuationSettings=/PlotTwist/Items/RiftGate/Assets/Sounds/Settings/ATT_AmberVine
 maxDistance=6400
 numChannels=2
 outputFormat=Stereo
 sampleRate=48000
 soundClassObject=&Fort_WeaponMelee
 volume=0.4
}
