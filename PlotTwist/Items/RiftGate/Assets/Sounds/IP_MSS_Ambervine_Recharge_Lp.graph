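# Looping stereo MetaSound source for the Ambervine recharge sound in the RiftGate item.
# RefuelPct is mapped to volume and pitch offsets that rise with the refuel level, stored in variables shared by both layers.
# Main and Sweetener layers each run MSS_PlayRandom_Loop with their own base values and random spreads, then mix to stereo.
# Main defaults to a Storm King lightning loop; Sweetener has no default waves, and Character inputs are unused.
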MetaSoundSource /PlotTwist/Items/RiftGate/Assets/Sounds/IP_MSS_Ambervine_Recharge_Lp
inputs [
 Character.Affiliation {type=Enum:FortAudioAffiliation}
 Character.Health {type=Float}
 Character.UniqueID {type=Int32}
 Main.Db {type=Float}
 Main.Randomization.DbSpread {type=Float}
 Main.Randomization.SemitoneSpread {type=Float}
 Main.Semitone {type=Float}
 Main.Waves {type=WaveAsset:Array}
 RefuelPct {type=Float}
 Sweetener.Db {type=Float}
 Sweetener.Randomize.DbSpread {type=Float}
 Sweetener.Randomize.SemitoneSpread {type=Float}
 Sweetener.Semitone {type=Float}
 Sweetener.Waves {type=WaveAsset:Array}
 UE.Source.OnPlay {type=Trigger}
]
interfaces [Fort.Character:1.0, UE.OutputFormat.Stereo:1.0, UE.Source:1.1]
outputs [
 "UE.OutputFormat.Stereo.Audio:0" {type=Audio}
 "UE.OutputFormat.Stereo.Audio:1" {type=Audio}
]
pages [Default {
  input_defaults {
   Character.Affiliation=0
   Character.Health=0
   Character.UniqueID=0
   Main.Db=-24
   Main.Randomization.DbSpread=4
   Main.Randomization.SemitoneSpread=1
   Main.Semitone=12
   RefuelPct=1
   Sweetener.Db=0
   Sweetener.Randomize.DbSpread=2
   Sweetener.Randomize.SemitoneSpread=1
   Sweetener.Semitone=0
   UE.Source.OnPlay=null
   Main.Waves [/Game/Sounds/Fort_Monster_Storm_King/SFX/Final_SFX/StormKing_LightningSummons_Lightning_Prct_Loop_03]
   Sweetener.Waves []
  }
  nodes [
   n1 = VariableAccessor.Float(Variable=VariableMutator.Float(
     Variable=InitVariable.Float().Variable
     Value=MapRange.MapRange:Float(In=^RefuelPct) {also {defaults {Out Range A=-8, Out Range B=0}}}.Out Value
    ).Variable)
   n2 = VariableAccessor.Float(Variable=VariableMutator.Float(Variable=InitVariable.Float().Variable, Value=MapRange.MapRange:Float(In=^RefuelPct) {also {defaults {Out Range B=12}}}.Out Value).Variable)
   n3 = /Metasound/DefaultAssets/Sources/MSS_PlayRandom_Loop(
    VolumeRandomSpread=^Main.Randomization.DbSpread
    PitchRandomSpread=^Main.Randomization.SemitoneSpread
    Sounds=^Main.Waves
    UE.Source.OnPlay=^UE.Source.OnPlay
    VolumeBase=UE.Add:Float(PrimaryOperand=^Main.Db, AdditionalOperands=@n1.Value).Out
    PitchBase=UE.Add:Float(PrimaryOperand=^Main.Semitone, AdditionalOperands=@n2.Value).Out
   )
   n4 = &MSS_PlayRandom_Loop(
    VolumeRandomSpread=^Sweetener.Randomize.DbSpread
    PitchRandomSpread=^Sweetener.Randomize.SemitoneSpread
    Sounds=^Sweetener.Waves
    UE.Source.OnPlay=^UE.Source.OnPlay
    VolumeBase=UE.Add:Float(PrimaryOperand=^Sweetener.Db, AdditionalOperands=VariableAccessor.Float(Variable=@n1.Variable).Value).Out
    PitchBase=UE.Add:Float(PrimaryOperand=^Sweetener.Semitone, AdditionalOperands=VariableAccessor.Float(Variable=@n2.Variable).Value).Out
   )
   AudioMixerAudioMixerStereo2_0 = "AudioMixer.Audio Mixer (Stereo, 2)"(
    In 0 L=@n3."UE.OutputFormat.Stereo.Audio:0"
    In 0 R=@n3."UE.OutputFormat.Stereo.Audio:1"
    In 1 L=@n4."UE.OutputFormat.Stereo.Audio:0"
    In 1 R=@n4."UE.OutputFormat.Stereo.Audio:1"
   )
  ]
  links [
   {dst="@UE.OutputFormat.Stereo.Audio:0", src=AudioMixerAudioMixerStereo2_0.Out L}
   {dst="@UE.OutputFormat.Stereo.Audio:1", src=AudioMixerAudioMixerStereo2_0.Out R}
  ]
 }]
properties {
 assetImportData=&self:AssetImportData
 attenuationSettings=/PlotTwist/Items/RiftGate/Assets/Sounds/Settings/ATT_AmberVine_Beam
 bLooping=true
 duration=10000
 maxDistance=8000
 numChannels=2
 outputFormat=Stereo
 sampleRate=48000
 soundClassObject=&Fort_WeaponMelee
}
